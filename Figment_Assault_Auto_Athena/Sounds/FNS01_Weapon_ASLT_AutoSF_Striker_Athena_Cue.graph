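# Sound cue for the single-shot fire of the Striker auto assault rifle in Athena, layering close, medium and distant recordings by distance.
# Each layer has its own attenuation; close uses the third-person fire sound class, while volume-boosted medium and distant use the distant-fire class.
# Close and medium mix through a shared modulator, then join distant in the root mixer.
# Global and local gunfire concurrency sets govern simultaneous playback.
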SoundCue /Figment_Assault_Auto_Athena/Sounds/FNS01_Weapon_ASLT_AutoSF_Striker_Athena_Cue
root=Mixer_2
nodes [
 {id=Attenuation_3, children [WavePlayer_2], properties {attenuationSettings=&Fort_Weapon_Projectile_Close_Athena_Att}}
 {
  id=Attenuation_4
  children [Modulator_0]
  properties {attenuationSettings=&Fort_Weapon_Projectile_Distant_Mono_Athena_Att}
 }
 {
  id=Attenuation_5
  children [Modulator_2]
  properties {attenuationSettings=&Fort_Weapon_Projectile_Medium_Mono_Athena_Att}
 }
 {id=Mixer_2, children [Modulator_3,SoundClass_4], properties {inputVolume [1,0.04]}}
 {id=Mixer_3, children [SoundClass_3,SoundClass_5], properties {inputVolume [0.5,0.16]}}
 {id=Modulator_0, children [WavePlayer_0], properties {volumeMax=2, volumeMin=2}}
 {id=Modulator_2, children [WavePlayer_1], properties {volumeMax=2, volumeMin=2}}
 {id=Modulator_3, children [Mixer_3]}
 {
  id=SoundClass_3
  children [Attenuation_3]
  properties {soundClassOverride=&Fort_WeaponRanged_Fire_3P_PSM}
 }
 {
  id=SoundClass_4
  children [Attenuation_4]
  properties {soundClassOverride=&Fort_WeaponRanged_Fire_Distant}
 }
 {
  id=SoundClass_5
  children [Attenuation_5]
  properties {soundClassOverride=&Fort_WeaponRanged_Fire_Distant}
 }
 {id=WavePlayer_0, properties {soundWaveAssetPtr=&/sfx_striker_singlefire_distant}}
 {id=WavePlayer_1, properties {soundWaveAssetPtr=&/sfx_striker_singlefire_medium}}
 {id=WavePlayer_2, properties {soundWaveAssetPtr=&/sfx_striker_singlefire_close}}
]
properties {
 bDebug=true
 maxDistance=27500
 soundClassObject=&Fort_WeaponRanged_Fire_3P_PSM
 volumeMultiplier=0.9
 concurrencyOverrides {bLimitToOwner=true, maxCount=3, resolutionRule=StopOldest}
 concurrencySet [&Weapon_Gunfire_Global_Conc, &Weapon_Gunfire_Local_Conc]
}
