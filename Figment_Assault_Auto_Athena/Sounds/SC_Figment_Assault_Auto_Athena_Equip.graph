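# Sound cue for the equip sound of the Figment auto assault rifle in Athena.
# It delays briefly by a fixed amount, then plays a single reused auto rifle cock wave from the shared reload elements.
# It routes through the weapon foley sound class with reload attenuation, overrides concurrency and has its debug flag turned on.
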SoundCue /Figment_Assault_Auto_Athena/Sounds/SC_Figment_Assault_Auto_Athena_Equip
root=Delay_0
nodes [
 {id=Delay_0, children [WavePlayer_27], properties {delayMax=0.1, delayMin=0.1}}
 {id=WavePlayer_27, properties {soundWaveAssetPtr=/Game/Sounds/Fort_Weapons/AutoRifle/Reload_Elements/rifle_auto_single_cock}}
]
properties {
 attenuationSettings=&Fort_Weapon_Reload_Att
 bDebug=true
 bOverrideConcurrency=true
 maxDistance=2250
 soundClassObject=&Fort_WeaponFoley
 volumeMultiplier=0.626
}
